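# Editor utility function library that generates lower-detail LODs for spline static meshes with Geometry Script.
# GenerateSplineMeshLOD groups vertices by rounded UV0 coordinate, marks alternate rings, welds edges, then MergeVerts collapses ring vertices into UV-aligned neighbours.
# It writes the result to the next LOD and disables that LOD's lightmap UV generation.
# GetUVDirection infers U or V alignment from distinct coordinate counts; TestUVs copies an LOD unaltered to the next.
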EditorUtilityBlueprint /Game/NeverCook/TechArt/SplineMeshTools/EFL_SplineMeshTools
graphs [
 GenerateSplineMeshLOD {nodes [
   GenerateSplineMeshLOD!() -> (then, FromLOD:int, StaticMeshAsset:Static Mesh Object Reference, AlignedToU:bool=true) {is=title K2Node_FunctionEntry}
   > n1 = GetGlobalMeshPool!(execute) -> (then, ReturnValue:Dynamic Mesh Pool Object Reference)
   > n2 = Request and Release Compute Mesh!(Exec:Exec, Mesh Pool:Dynamic Mesh Pool Object Reference=@n1) -> (With Mesh:Exec, Mesh:Dynamic Mesh Object Reference, After Release:Exec) {is=title K2Node_MacroInstance}
   > Set l_mesh(l_mesh:Dynamic Mesh Object Reference=@n2) -> (Output_Get:Dynamic Mesh Object Reference)
   > n3 = Copy Mesh From Static Mesh!(
    execute
    FromStaticMeshAsset:Static Mesh Object Reference=Get StaticMeshAsset() -> (StaticMeshAsset:Static Mesh Object Reference)
    ToDynamicMesh:Dynamic Mesh Object Reference=Get l_mesh() -> (l_mesh:Dynamic Mesh Object Reference)
    AssetOptions:Geometry Script Copy Mesh From Asset Options Structure=Make Geometry Script Copy Mesh from Asset Options!(bApplyBuildSettings:bool=True, bRequestTangents:bool=True, bIgnoreRemoveDegenerates:bool=True, bUseBuildScale:bool=True) -> (GeometryScriptCopyMeshFromAssetOptions:Geometry Script Copy Mesh From Asset Options Structure) {is=title K2Node_MakeStruct}
    RequestedLOD:Geometry Script Mesh Read LOD Structure=Make Geometry Script Mesh Read LOD!(LODType:EGeometryScriptLODType Enum=RenderData, LODIndex:int=Get FromLOD() -> (FromLOD:int=0)) -> (GeometryScriptMeshReadLOD:Geometry Script Mesh Read LOD Structure) {is=title K2Node_MakeStruct}
    bUseSectionMaterials:bool=true
   ) -> (Failure:Exec, Success:Exec, ReturnValue:Dynamic Mesh Object Reference)
   n4 = Sequence!(execute=@n3.Success) -> (then_0, then_1, then_2:Exec, then_3:Exec, then_4:Exec)
   > n5 = GetAllVertexIDs!(execute, TargetMesh:Dynamic Mesh Object Reference=Get l_mesh()) -> (then, VertexIDList:Index List Structure, bHasVertexIDGaps:bool, ReturnValue:Dynamic Mesh Object Reference)
   > n6 = ConvertIndexListToArray!(execute, IndexList:Index List Structure=@n5) -> (then, IndexArray:Array of Integers)
   > n7 = For Each Loop(Array:Array of Integers=@n6) -> (Array Element:int)
   > n8 = Set l_index(l_index:int=@n7) -> (Output_Get:int=0)
   > n9 = GetMeshUVElementPosition!(execute, TargetMesh:Dynamic Mesh Object Reference=Get l_mesh(), UVSetIndex:int=0, ElementID:int=@n8) -> (then, UVPosition:Vector 2D Structure, bIsValidElementID:bool, ReturnValue:Dynamic Mesh Object Reference)
   > Set l_uv(l_uv:Vector 2D Structure=@n9) -> (Output_Get:Vector 2D Structure)
   n10 = Make Literal Float(Value=1000)
   n11 = BreakVector2D(InVec=Get l_uv() -> (l_uv:Vector 2D Structure))
   n12 = float * float(A=Select(Option 0:double=@n11.Y, Option 1:double=@n11, Index=Get AlignedToU() -> (AlignedToU:bool)) -> (ReturnValue:double), B=@n10) {is=title K2Node_PromotableOperator}
   > Set l_roundedU(l_roundedU:double=float / float(A:int=Round(A=@n12) {is=title K2Node_CallFunction}, B=@n10)) -> (Output_Get:double)
   > n13 = Find!(
    execute:Exec
    TargetMap:"Map of Float (double-precision)s to Spline UVMap Struct Structures"=Get l_UMap() -> (l_UMap:"Map of Float (double-precision)s to Spline UVMap Struct Structures")
    Key:"Float (double-precision) (by ref)"=Get l_roundedU() -> (l_roundedU:double)
   ) -> (then:Exec, Value:Spline UVMap Struct Structure, ReturnValue)
   > n14 = Sequence!(execute) -> (then_0, then_1)
   > n15 = Branch(Condition=@n13.ReturnValue)
   > n16 = Set l_vertIDs(l_vertIDs:Array of Integers=Break Spline UVMap Struct!(SplineUVMapStruct:"Spline UVMap Struct Structure (by ref)"=@n13) -> (VertexIDs_7_#1:Array of Integers, UV_5_#2:Vector 2D Structure) {is=title K2Node_BreakStruct}) -> (Output_Get:Array of Integers)
   > Add(TargetArray:Array of Integers=@n16, NewItem:"Integer (by ref)"=Get l_index() -> (l_index:int=0)) {is=title K2Node_CallArrayFunction}
   Set l_vertIDs(execute=@n15.else, l_vertIDs=Make Array("[0]":int=Get l_index() -> (l_index=0)) -> (Array:Array of Integers))
   Add!(
    execute=@n14.then_1
    TargetMap:"Map of Float (double-precision)s to Spline UVMap Struct Structures"=Get l_UMap()
    Key:"Float (double-precision) (by ref)"=Get l_roundedU()
    Value:"Spline UVMap Struct Structure (by ref)"=Make Spline UVMap Struct!(
     VertexIDs_7_#1:Array of Integers=Get l_vertIDs() -> (l_vertIDs:Array of Integers)
     UV_5_#2:Vector 2D Structure=Get l_uv()
    ) -> (SplineUVMapStruct:Spline UVMap Struct Structure) {is=title K2Node_MakeStruct}
   ) -> (then) {is=title K2Node_CallFunction}
   n17 = Values!(execute=@n4.then_1, TargetMap:"Map of Float (double-precision)s to Spline UVMap Struct Structures"=Get l_UMap()) -> (then, Values:Array of Spline UVMap Struct Structures)
   > n18 = For Each Loop(Array:Array of Spline UVMap Struct Structures=@n17) -> (Array Element:Spline UVMap Struct Structure)
   n19 = Break Spline UVMap Struct(SplineUVMapStruct=@n18)
   > Branch(Condition="integer >= integer"(A=Length(TargetArray:Array of Integers=@n19), B=2))
   > n20 = Set l_delete(l_delete:bool=NOT Boolean(A=Get l_delete() -> (l_delete:bool))) -> (Output_Get:bool)
   > Branch(Condition=@n20)
   > For Each Loop(Array:Array of Integers=@n19) -> (Array Element:int)
   > Add(
    TargetArray:Array of Spline UVMap Struct Structures=Get l_structsToRemove() -> (l_structsToRemove:Array of Spline UVMap Struct Structures)
    NewItem:"Spline UVMap Struct Structure (by ref)"=@n18
   ) {is=title K2Node_CallArrayFunction}
   WeldMeshEdges!(
    execute=@n4.then_2
    TargetMesh:Dynamic Mesh Object Reference=Get l_mesh()
    WeldOptions:Geometry Script Weld Edges Options Structure=Make Geometry Script Weld Edges Options!(Tolerance:float=0.01, bOnlyUniquePairs:bool=false) -> (GeometryScriptWeldEdgesOptions:Geometry Script Weld Edges Options Structure) {is=title K2Node_MakeStruct}
   ) -> (then, ReturnValue:Dynamic Mesh Object Reference)
   > n21 = For Each Loop(Array:Array of Spline UVMap Struct Structures=Get l_structsToRemove()) -> (Array Element:Spline UVMap Struct Structure)
   n22 = Break Spline UVMap Struct(SplineUVMapStruct=@n21)
   > n23 = For Each Loop(Array:Array of Integers=@n22) -> (Array Element:int)
   > MergeVerts!(
    execute
    VertToRemove:int=@n23
    VertsToFilter:Array of Integers=@n22
    TargetMesh:Dynamic Mesh Object Reference=Get l_mesh()
    AlignedToU:bool=Get AlignedToU()
    __WorldContext:Object Reference
   ) -> (then)
   n24 = Get FromLOD() -> (FromLOD=0)
   n25 = CopyMeshToStaticMesh!(
    execute=[@n4.then_4, @n31.then]
    FromDynamicMesh:Dynamic Mesh Object Reference=Get l_mesh()
    ToStaticMeshAsset:Static Mesh Object Reference=Get StaticMeshAsset()
    Options:Geometry Script Copy Mesh To Asset Options Structure=Make Geometry Script Copy Mesh to Asset Options!(
     NewMaterialUsage:ECopyMeshToAssetUseNewMaterialsOptions Enum=ReplaceAssetMaterials
     bCleanAssignedMaterials:bool=False
     GenerateLightmapUVs:EGeometryScriptGenerateLightmapUVOptions Enum=MatchTargetLODSetting
    ) -> (GeometryScriptCopyMeshToAssetOptions:Geometry Script Copy Mesh To Asset Options Structure) {is=title K2Node_MakeStruct}
    TargetLOD:Geometry Script Mesh Write LOD Structure=Make Geometry Script Mesh Write LOD!(bWriteHiResSource:bool=False, LODIndex:int=int + int(A=@n24)) -> (GeometryScriptMeshWriteLOD:Geometry Script Mesh Write LOD Structure) {is=title K2Node_MakeStruct}
    bUseSectionMaterials:bool=true
   ) -> (Failure:Exec, Success:Exec, ReturnValue:Dynamic Mesh Object Reference)
   Do Once(execute=@n25.Success)
   n26 = Get StaticMeshEditorSubsystem()
   n27 = int + int(A=Get FromLOD() -> (FromLOD=0))
   n28 = Get StaticMeshAsset()
   > n29 = GetLodBuildSettings!(execute, self:Static Mesh Editor Subsystem Object Reference=@n26, StaticMesh:Static Mesh Object Reference=@n28, LodIndex:int=@n27) -> (then, OutBuildOptions:Mesh Build Settings Structure)
   > n30 = Set members in Mesh Build Settings!(execute:Exec, StructRef:"Mesh Build Settings Structure (by ref)"=@n29, bGenerateLightmapUVs:bool=false) -> (then:Exec, StructOut:"Mesh Build Settings Structure (by ref)") {is=title K2Node_SetFieldsInStruct}
   > n31 = SetLodBuildSettings!(
    execute
    self:Static Mesh Editor Subsystem Object Reference=@n26
    StaticMesh:Static Mesh Object Reference=@n28
    LodIndex:int=@n27
    BuildOptions:"Mesh Build Settings Structure (by ref)"=@n30
   ) -> (then)
  ]}
 GetUVDirection {nodes [
   GetUVDirection!() -> (then, StaticMeshAsset:Static Mesh Object Reference) {is=title K2Node_FunctionEntry}
   > n1 = GetGlobalMeshPool()
   > n2 = Request and Release Compute Mesh(Mesh Pool=@n1)
   > Set l_mesh(l_mesh=@n2)
   > n3 = Copy Mesh From Static Mesh(
    FromStaticMeshAsset=Get StaticMeshAsset()
    ToDynamicMesh=Get l_mesh()
    RequestedLOD=Make Geometry Script Mesh Read LOD(LODType=RenderData, LODIndex=0)
    bUseSectionMaterials=true
   )
   n4 = Sequence(execute=@n3.Success)
   > n5 = GetAllVertexIDs(TargetMesh=Get l_mesh())
   > n6 = ConvertIndexListToArray(IndexList=@n5)
   > n7 = For Each Loop(Array=@n6)
   > n8 = Set l_index(l_index=@n7) -> (Output_Get=0)
   > n9 = GetMeshUVElementPosition(TargetMesh=Get l_mesh(), UVSetIndex=0, ElementID=@n8)
   > Set l_uv(l_uv=@n9)
   > n10 = Sequence()
   n11 = Make Literal Float(Value=1000)
   n12 = float * float(A=BreakVector2D(InVec=Get l_uv()), B=@n11)
   > Set l_roundedValue(l_roundedValue:double=float / float(A=Round(A=@n12), B=@n11)) -> (Output_Get:double)
   > n13 = Find(TargetMap=Get l_UMap(), Key=Get l_roundedValue() -> (l_roundedValue:double))
   > n14 = Sequence()
   > n15 = Branch(Condition=@n13.ReturnValue)
   > n16 = Set l_vertIDs(l_vertIDs=Break Spline UVMap Struct(SplineUVMapStruct=@n13))
   > Add(TargetArray:Array of Integers=@n16, NewItem:"Integer (by ref)"=Get l_index() -> (l_index=0))
   Set l_vertIDs(execute=@n15.else, l_vertIDs=Make Array("[0]"=Get l_index() -> (l_index=0)))
   Add(
    execute=@n14.then_1
    TargetMap=Get l_UMap()
    Key=Get l_roundedValue()
    Value=Make Spline UVMap Struct(VertexIDs_7_#1=Get l_vertIDs(), UV_5_#2=Get l_uv())
   ) {is=title K2Node_CallFunction}
   n17 = Make Literal Float(Value=1000)
   n18 = float * float(A=BreakVector2D(InVec=Get l_uv()).Y, B=@n17)
   Set l_roundedValue(execute=@n10.then_1, l_roundedValue=float / float(A=Round(A=@n18), B=@n17))
   > n19 = Find(TargetMap=Get l_VMap() -> (l_VMap:"Map of Float (double-precision)s to Spline UVMap Struct Structures"), Key=Get l_roundedValue())
   > n20 = Sequence()
   > n21 = Branch(Condition=@n19.ReturnValue)
   > n22 = Set l_vertIDs(l_vertIDs=Break Spline UVMap Struct(SplineUVMapStruct=@n19))
   > Add(TargetArray=@n22, NewItem=Get l_index() -> (l_index=0)) {is=title K2Node_CallArrayFunction}
   Set l_vertIDs(execute=@n21.else, l_vertIDs=Make Array("[0]"=Get l_index() -> (l_index=0)))
   Add(
    execute=@n20.then_1
    TargetMap=Get l_VMap()
    Key=Get l_roundedValue()
    Value=Make Spline UVMap Struct(VertexIDs_7_#1=Get l_vertIDs(), UV_5_#2=Get l_uv())
   ) {is=title K2Node_CallFunction}
   n23 = Keys!(execute=@n4.then_1, TargetMap:"Map of Float (double-precision)s to Spline UVMap Struct Structures"=Get l_UMap()) -> (then, Keys:"Array of Float (double-precision)s")
   > n24 = Keys(TargetMap=Get l_VMap())
   > n25 = Branch(Condition="integer > integer"(A=Length(TargetArray:"Array of Float (double-precision)s"=@n24), B=Length(TargetArray=@n23)))
   > Return Node!(execute, AlignedToU:bool=false)
   Return Node(execute=@n25.else, AlignedToU=true)
  ]}
 MergeVerts {nodes [
   n1 = MergeVerts!() -> (then, VertToRemove:int, VertsToFilter:Array of Integers, TargetMesh:Dynamic Mesh Object Reference, AlignedToU:bool=true) {is=title K2Node_FunctionEntry}
   > n2 = GetVertexConnectedVertices!(execute, TargetMesh:Dynamic Mesh Object Reference=@n1.TargetMesh, VertexID:int=Get VertToRemove() -> (VertToRemove:int=0)) -> (then, Vertices:Array of Integers, ReturnValue:Dynamic Mesh Object Reference)
   > Set l_connected(l_connected:Array of Integers=@n2) -> (Output_Get:Array of Integers)
   > n3 = Sequence!(execute) -> (then_0, then_1, then_2:Exec)
   > n4 = For Each Loop(Array=Get VertsToFilter() -> (VertsToFilter:Array of Integers))
   > Remove Item!(execute, TargetArray:Array of Integers=Get l_connected() -> (l_connected:Array of Integers), Item:"Integer (by ref)"=@n4) -> (then, ReturnValue:bool)
   n5 = Get l_connected()
   > Branch(Condition="integer < integer"(A=Length(TargetArray:Array of Integers=@n5), B=3))
   > Return Node!(execute)
   n6 = For Each Loop with Break!(Exec:Exec=@n3.then_1, Array:Array of Integers=Get l_connected(), Break:Exec=@n15.then) -> (LoopBody:Exec, Array Element:int, Array Index:int, Completed:Exec)
   n7 = Get TargetMesh() -> (TargetMesh:Dynamic Mesh Object Reference)
   > n8 = GetMeshUVElementPosition(TargetMesh=@n7, UVSetIndex=0, ElementID=@n6)
   > n9 = GetMeshUVElementPosition(TargetMesh=@n7, UVSetIndex=0, ElementID=Get VertToRemove() -> (VertToRemove=0))
   n10 = BreakVector2D(InVec=@n8)
   n11 = Get AlignedToU()
   n12 = Select(Option 0=@n10, Option 1=@n10.Y, Index=@n11)
   n13 = BreakVector2D(InVec=@n9)
   n14 = Select(Option 0=@n13, Option 1=@n13.Y, Index=@n11)
   > Branch(Condition="Equal (Integer)"(A=Round(A=@n12), B=Round(A=@n14)))
   > n15 = Set l_vertToKeep(l_vertToKeep:int=@n6) -> (Output_Get:int=0)
   Branch(execute=@n3.then_2, Condition="integer > integer"(A=Get l_vertToKeep() -> (l_vertToKeep:int=0), B=-1))
   > MergeMeshVertexPair!(
    execute
    TargetMesh:Dynamic Mesh Object Reference=Get TargetMesh()
    VertexKeep:int=Get l_vertToKeep() -> (l_vertToKeep=0)
    VertexDiscard:int=Get VertToRemove() -> (VertToRemove=0)
    Options:Geometry Script Merge Vertex Options Structure=Make Geometry Script Merge Vertex Options!(bOnlyBoundary:bool=True, bAllowNonBoundaryBowties:bool=False) -> (GeometryScriptMergeVertexOptions:Geometry Script Merge Vertex Options Structure) {is=title K2Node_MakeStruct}
    InterpParam:double=0
    bDeferChangeNotifications:bool=false
   ) -> (then, bSuccess:bool, ReturnValue:Dynamic Mesh Object Reference)
  ]}
 TestUVs {nodes [
   TestUVs!() -> (then, FromLOD:int, StaticMeshAsset:Static Mesh Object Reference) {is=title K2Node_FunctionEntry}
   > n1 = GetGlobalMeshPool()
   > n2 = Request and Release Compute Mesh(Mesh Pool=@n1)
   > Set l_mesh(l_mesh=@n2)
   n3 = Get FromLOD() -> (FromLOD=0)
   n4 = Get l_mesh()
   n5 = Get StaticMeshAsset()
   > n6 = Copy Mesh From Static Mesh(
    FromStaticMeshAsset=@n5
    ToDynamicMesh=@n4
    RequestedLOD=Make Geometry Script Mesh Read LOD(LODType=SourceModel, LODIndex=@n3)
    bUseSectionMaterials=true
   )
   CopyMeshToStaticMesh(
    execute=@n6.Success
    FromDynamicMesh=@n4
    ToStaticMeshAsset=@n5
    Options=Make Geometry Script Copy Mesh to Asset Options(NewMaterialUsage=ReplaceAssetMaterials, bCleanAssignedMaterials=False, GenerateLightmapUVs=MatchTargetLODSetting)
    TargetLOD=Make Geometry Script Mesh Write LOD(bWriteHiResSource=False, LODIndex=int + int(A=@n3))
    bUseSectionMaterials=true
   )
  ]}
]
variables []
